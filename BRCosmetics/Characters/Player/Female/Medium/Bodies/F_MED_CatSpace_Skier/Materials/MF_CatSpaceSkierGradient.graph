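# Material function for the CatSpace Skier character cosmetic that layers an animated gradient and pattern over incoming material attributes, blended by an Alpha input.
# Pre-skinned position over local bounds gives a gradient coordinate that stays fixed through animation.
# A time-scrolled gradient texture and a pannable XZ-projected triangle pattern, with a two-ramp Z mask, feed color and emissive adjustments.
# Scalar biases offset specular and roughness.
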MaterialFunction /BRCosmetics/Characters/Player/Female/Medium/Bodies/F_MED_CatSpace_Skier/Materials/MF_CatSpaceSkierGradient
graph {nodes [
  n1 = SetMaterialAttributes!(
   MaterialAttributes=NamedRerouteUsage()
   EmissiveColor=NamedRerouteUsage()
   BaseColor=NamedRerouteUsage()
   Specular=NamedRerouteUsage()
   Roughness=NamedRerouteUsage()
  ) {attributeSetTypes [
    #1
    #2
    #3
    #4
   ]}
  FunctionOutput(None=BlendMaterialAttributes(A=NamedRerouteUsage(), B=@n1, Alpha=NamedRerouteUsage()))
  n2 = GetMaterialAttributes(None=NamedRerouteDeclaration(None=FunctionInput() {inputName=Material, inputType=FunctionInput_MaterialAttributes}) {name=InMaterial, nodeColor=<rgba 0.1719 1 0 1>}) {attributeGetTypes [
    #1
    #2
    #3
    #4
   ]}
  NamedRerouteDeclaration(None=@n2.Roughness) {name=InRoughness, nodeColor=<rgba 0 0.7656 1 1>}
  NamedRerouteDeclaration(None=@n2.Specular) {name=InSpec, nodeColor=<rgba 0 0.7656 1 1>}
  NamedRerouteDeclaration(None=@n2.BaseColor) {name=InBaseColor, nodeColor=<rgba 1 0 0.1875 1>}
  NamedRerouteDeclaration(None=@n2.EmissiveColor) {name=InEmissive, nodeColor=<rgba 0 0.9297 1 1>}
  n3 = NamedRerouteUsage()
  n4 = ScalarParameter() {defaultValue=-0.3, group=CatSpaceGradient, parameterName=RoughnessBias}
  NamedRerouteDeclaration(None=Saturate(None=Add(A=@n3, B=@n4))) {name=OutRoughness, nodeColor=<rgba 0 1 0.0625 1>}
  n5 = NamedRerouteUsage()
  n6 = ScalarParameter() {defaultValue=0.5, group=CatSpaceGradient, parameterName=SpecBias}
  NamedRerouteDeclaration(None=Saturate(None=Add(A=@n5, B=@n6))) {name=OutSpec, nodeColor=<rgba 0 0.6953 1 1>}
  n7 = NamedRerouteUsage()
  n8 = NamedRerouteUsage()
  n9 = ScalarParameter() {defaultValue=0.8, group=CatSpaceGradient, parameterName=GradientEmissiveStrength}
  NamedRerouteDeclaration(None=Multiply(A=LinearInterpolate(A=@n7, B=@n8, Alpha=@n9), B=ScalarParameter() {defaultValue=0.4, group=CatSpaceGradient, parameterName=EmissiveStrength})) {name=GradientEmissive, nodeColor=<rgba 1 0 0.3281 1>}
  n10 = Multiply(A=ComponentMask(None=NamedRerouteUsage()) {b=true, r=true}, B=Constant2Vector() {g=-1, r=1})
  n11 = ScalarParameter() {defaultValue=8, group=CatSpaceGradient, parameterName=PatternTileDivisor}
  n12 = ScalarParameter() {group=CatSpaceGradient, parameterName=PatternPanning_X}
  n13 = ScalarParameter() {group=CatSpaceGradient, parameterName=PatternPanning_Y}
  n14 = TextureSampleParameter2D(UVs=Panner(Coordinate=Divide(A=@n10, B=@n11) {constB=100}, Speed=AppendVector(A=@n12, B=@n13))) {
   group=CatSpaceGradient
   parameterName=PatternTexture
   samplerType=SAMPLERTYPE_Masks
   texture=/Game/Characters/Player/Female/Medium/Bodies/F_MED_Neoncat/Textures/Triangles_M
  }
  n15 = ScalarParameter() {defaultValue=0.5, group=CatSpaceGradient, parameterName=PatternStrength}
  n16 = NamedRerouteUsage()
  NamedRerouteDeclaration(None=Multiply(A=@n14.R, B=Multiply(A=@n15, B=@n16))) {name=Pattern, nodeColor=<rgba 0 1 0.3438 1>}
  NamedRerouteDeclaration(None=LinearInterpolate(
    A=NamedRerouteUsage()
    B=NamedRerouteUsage()
    Alpha=ScalarParameter() {defaultValue=0.7, group=CatSpaceGradient, parameterName=GradientBaseColorStrength}
   )) {name=OutBaseColor, nodeColor=<rgba 1 0 0.2812 1>}
  NamedRerouteDeclaration(None=Add(A=NamedRerouteUsage(), B=NamedRerouteUsage())) {name=OutEmissive, nodeColor=<rgba 0.2422 0 1 1>}
  n17 = ComponentMask(None=NamedRerouteUsage()) {b=true}
  n18 = Multiply(
   A=Add(A=OneMinus(None=@n17), B=ScalarParameter() {defaultValue=-0.15, group=CatSpaceGradient, parameterName=ZGradientBias_01})
   B=ScalarParameter() {defaultValue=1, group=CatSpaceGradient, parameterName=ZGradientTile_01}
  )
  n19 = NamedRerouteUsage()
  n20 = Multiply(
   A=Add(A=ComponentMask(None=@n19) {b=true}, B=ScalarParameter() {defaultValue=-0.5374, group=CatSpaceGradient, parameterName=ZGradientBias_02})
   B=ScalarParameter() {defaultValue=5, group=CatSpaceGradient, parameterName=ZGradientTile_02}
  )
  NamedRerouteDeclaration(None=Multiply(A=Saturate(None=@n18), B=Saturate(None=@n20))) {name=PatternGradientMask, nodeColor=<rgba 0 1 0.25 1>}
  n21 = VertexInterpolator(VS=PreSkinnedPosition())
  n22 = PreSkinnedLocalBounds()
  NamedRerouteDeclaration(None=Divide(A=NamedRerouteDeclaration(None=@n21.PS) {name=PreskinnedPos, nodeColor=<rgba 0 0.1328 1 1>}, B=VertexInterpolator(VS=@n22.Extents).PS)) {name=PreskinnedGradient, nodeColor=<rgba 1 0 0.6562 1>}
  n23 = FunctionInput() {bUsePreviewValueAsDefault=true, inputName=Alpha, inputType=FunctionInput_Scalar, sortPriority=1, previewValue=<1 0 0 1>}
  NamedRerouteDeclaration(None=@n23) {name=InAlpha, nodeColor=<rgba 1 0 0.0937 1>}
  n24 = Time()
  n25 = ScalarParameter() {defaultValue=-0.2, group=CatSpaceGradient, parameterName=GradientSpeed}
  n26 = ComponentMask(None=NamedRerouteUsage()) {b=true}
  n27 = ScalarParameter() {defaultValue=1, group=CatSpaceGradient, parameterName=GradientTile}
  n28 = Frac(None=Add(A=Multiply(A=@n24, B=@n25), B=Multiply(A=@n26, B=@n27) {constB=2}))
  n29 = Constant() {r=0.5}
  NamedRerouteDeclaration(None=TextureSample(UVs=MaterialFunctionCall(X=@n28, Y=@n29) {materialFunction=&MakeFloat2}.Result) {texture=/BRCosmetics/Characters/Player/Female/Medium/Bodies/F_MED_CatSpace_Skier/Textures/T_CatSpaceSkier_Gradient}.RGB) {name=GradientSample, nodeColor=<rgba 0.5703 0 1 1>}
 ]}
