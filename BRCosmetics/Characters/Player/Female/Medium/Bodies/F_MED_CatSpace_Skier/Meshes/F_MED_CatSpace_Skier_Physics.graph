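# Physics asset for the F_MED_CatSpace_Skier female body, set up for secondary motion of its jacket bones.
# Five three-bone capsule chains (back left/right, center, front left/right) each constrain bone to bone up to spine_04.
# All joints share tight angular limits, with one swing axis wider than the other swing and twist, so segments mostly bend one way.
# Capsules on the pelvis and spine bones provide torso bodies alongside the chains.
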PhysicsAsset /BRCosmetics/Characters/Player/Female/Medium/Bodies/F_MED_CatSpace_Skier/Meshes/F_MED_CatSpace_Skier_Physics
bodies {
 dyn_jacket_bk_01_l [{shapeType=Capsule, length=6, radius=3, shapeName=dyn_jacket_bk_01_l_capsule, center=<2 -0.9002 0>}]
 dyn_jacket_bk_01_r [{shapeType=Capsule, length=6, radius=3, shapeName=dyn_jacket_bk_01_r_capsule, center=<-2 0.9 0>}]
 dyn_jacket_bk_02_l [{shapeType=Capsule, length=6, radius=3, shapeName=dyn_jacket_bk_02_l_capsule, center=<2 -0.9002 0>}]
 dyn_jacket_bk_02_r [{shapeType=Capsule, length=6, radius=3, shapeName=dyn_jacket_bk_02_r_capsule, center=<-2 0.9 0>}]
 dyn_jacket_bk_03_l [{shapeType=Capsule, length=6, radius=3, shapeName=dyn_jacket_bk_03_l_capsule, center=<2 -0.9002 0>}]
 dyn_jacket_bk_03_r [{shapeType=Capsule, length=6, radius=3, shapeName=dyn_jacket_bk_03_r_capsule, center=<-2 0.9 0>}]
 dyn_jacket_c_01 [{shapeType=Capsule, length=6, radius=3, shapeName=dyn_jacket_c_01_capsule, center=<2 -0.9002 0>}]
 dyn_jacket_c_02 [{shapeType=Capsule, length=6, radius=3, shapeName=dyn_jacket_c_02_capsule, center=<2 -0.9002 0>}]
 dyn_jacket_c_03 [{shapeType=Capsule, length=6, radius=3, shapeName=dyn_jacket_c_03_capsule, center=<2 -0.9002 0>}]
 dyn_jacket_fr_01_l [{shapeType=Capsule, length=6, radius=3, shapeName=dyn_jacket_fr_01_l_capsule, center=<2 -0.9002 0>}]
 dyn_jacket_fr_01_r [{shapeType=Capsule, length=6, radius=3, shapeName=dyn_jacket_fr_01_r_capsule, center=<-2 0.9 0>}]
 dyn_jacket_fr_02_l [{shapeType=Capsule, length=6, radius=3, shapeName=dyn_jacket_fr_02_l_capsule, center=<2 -0.9002 0>}]
 dyn_jacket_fr_02_r [{shapeType=Capsule, length=6, radius=3, shapeName=dyn_jacket_fr_02_r_capsule, center=<-2 0.9 0>}]
 dyn_jacket_fr_03_l [{shapeType=Capsule, length=6, radius=3, shapeName=dyn_jacket_fr_03_l_capsule, center=<2 -0.9002 0>}]
 dyn_jacket_fr_03_r [{shapeType=Capsule, length=6, radius=3, shapeName=dyn_jacket_fr_03_r_capsule, center=<-2 0.9 0>}]
 pelvis [{shapeType=Capsule, length=8.5802, radius=11.8153, shapeName=pelvis_capsule}]
 spine_01 [{shapeType=Capsule, length=0.505, radius=0.505, shapeName=spine_01_capsule}]
 spine_02 [{shapeType=Capsule, length=8.5013, radius=11.1048, shapeName=spine_02_capsule, center=<1.0067 1.0351 0>, rotation=<rot -90 180 180>}]
 spine_03 [{shapeType=Capsule, length=4.9108, radius=8.1246, shapeName=spine_03_capsule}]
 spine_04 [{shapeType=Capsule, length=2.5708, radius=12.0484, shapeName=spine_04_capsule}]
 spine_05 [{shapeType=Capsule, length=6.9743, radius=10.0663, shapeName=spine_05_capsule}]
}
constraints [
 {bone1Name=dyn_jacket_bk_01_l, bone2Name=spine_04, swing1LimitDegrees=14, swing2LimitDegrees=2, twistLimitDegrees=2}
 {bone1Name=dyn_jacket_bk_02_l, bone2Name=dyn_jacket_bk_01_l, swing1LimitDegrees=14, swing2LimitDegrees=2, twistLimitDegrees=2}
 {bone1Name=dyn_jacket_bk_03_l, bone2Name=dyn_jacket_bk_02_l, swing1LimitDegrees=14, swing2LimitDegrees=2, twistLimitDegrees=2}
 {bone1Name=dyn_jacket_bk_01_r, bone2Name=spine_04, swing1LimitDegrees=14, swing2LimitDegrees=2, twistLimitDegrees=2}
 {bone1Name=dyn_jacket_bk_02_r, bone2Name=dyn_jacket_bk_01_r, swing1LimitDegrees=14, swing2LimitDegrees=2, twistLimitDegrees=2}
 {bone1Name=dyn_jacket_bk_03_r, bone2Name=dyn_jacket_bk_02_r, swing1LimitDegrees=14, swing2LimitDegrees=2, twistLimitDegrees=2}
 {bone1Name=dyn_jacket_c_01, bone2Name=spine_04, swing1LimitDegrees=14, swing2LimitDegrees=2, twistLimitDegrees=2}
 {bone1Name=dyn_jacket_c_02, bone2Name=dyn_jacket_c_01, swing1LimitDegrees=14, swing2LimitDegrees=2, twistLimitDegrees=2}
 {bone1Name=dyn_jacket_c_03, bone2Name=dyn_jacket_c_02, swing1LimitDegrees=14, swing2LimitDegrees=2, twistLimitDegrees=2}
 {bone1Name=dyn_jacket_fr_01_l, bone2Name=spine_04, swing1LimitDegrees=14, swing2LimitDegrees=2, twistLimitDegrees=2}
 {bone1Name=dyn_jacket_fr_02_l, bone2Name=dyn_jacket_fr_01_l, swing1LimitDegrees=14, swing2LimitDegrees=2, twistLimitDegrees=2}
 {bone1Name=dyn_jacket_fr_03_l, bone2Name=dyn_jacket_fr_02_l, swing1LimitDegrees=14, swing2LimitDegrees=2, twistLimitDegrees=2}
 {bone1Name=dyn_jacket_fr_01_r, bone2Name=spine_04, swing1LimitDegrees=14, swing2LimitDegrees=2, twistLimitDegrees=2}
 {bone1Name=dyn_jacket_fr_02_r, bone2Name=dyn_jacket_fr_01_r, swing1LimitDegrees=14, swing2LimitDegrees=2, twistLimitDegrees=2}
 {bone1Name=dyn_jacket_fr_03_r, bone2Name=dyn_jacket_fr_02_r, swing1LimitDegrees=14, swing2LimitDegrees=2, twistLimitDegrees=2}
]
