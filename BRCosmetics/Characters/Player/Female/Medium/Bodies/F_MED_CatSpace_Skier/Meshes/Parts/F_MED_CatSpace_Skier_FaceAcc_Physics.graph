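# Physics asset for the face accessory part of the female medium CatSpace Skier player cosmetic, covering secondary motion of front hair and hood.
# Capsules on dynamic bones form left and right hair chains plus left, center and right hood chains, each running from a master bone through three segments.
# Tight swing and twist limits link each chain, stiffest at center hood.
# Unconstrained head, upper spine, clavicle and upper arm capsules provide body collision shapes.
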PhysicsAsset /BRCosmetics/Characters/Player/Female/Medium/Bodies/F_MED_CatSpace_Skier/Meshes/Parts/F_MED_CatSpace_Skier_FaceAcc_Physics
bodies {
 clavicle_l [{shapeType=Capsule, length=25.7946, radius=4.1925, shapeName=clavicle_l_capsule, rotation=<rot -90 0 0>}]
 clavicle_r [{shapeType=Capsule, length=25.7946, radius=4.1925, shapeName=clavicle_r_capsule, rotation=<rot -90 0 0>}]
 dyn_hair_front_01_l [{shapeType=Capsule, length=6, radius=2, shapeName=dyn_hair_front_01_l_capsule}]
 dyn_hair_front_01_r [{shapeType=Capsule, length=6, radius=2, shapeName=dyn_hair_front_01_r_capsule}]
 dyn_hair_front_02_l [{shapeType=Capsule, length=6, radius=2, shapeName=dyn_hair_front_02_l_capsule}]
 dyn_hair_front_02_r [{shapeType=Capsule, length=6, radius=2, shapeName=dyn_hair_front_02_r_capsule}]
 dyn_hair_front_03_l [{shapeType=Capsule, length=6, radius=2, shapeName=dyn_hair_front_03_l_capsule}]
 dyn_hair_front_03_r [{shapeType=Capsule, length=6, radius=2, shapeName=dyn_hair_front_03_r_capsule}]
 dyn_hair_front_mstr_l [{shapeType=Capsule, length=0.505, radius=0.505, shapeName=dyn_hair_front_mstr_l_capsule, rotation=<rot 0 -90 0>}]
 dyn_hair_front_mstr_r [{shapeType=Capsule, length=0.505, radius=0.505, shapeName=dyn_hair_front_mstr_r_capsule, rotation=<rot 0 -90 0>}]
 dyn_hood_01_c [{shapeType=Capsule, length=8, radius=4, shapeName=dyn_hood_01_c_capsule, center=<2.6883 -0.5859 0>, rotation=<rot 0 -12.2948 0>}]
 dyn_hood_01_l [{shapeType=Capsule, length=8, radius=4, shapeName=dyn_hood_01_l_capsule, center=<2.6883 -0.5859 0>, rotation=<rot 0 -12.2949 0>}]
 dyn_hood_01_r [{shapeType=Capsule, length=8, radius=4, shapeName=dyn_hood_01_r_capsule, center=<2.6883 -0.5859 0>, rotation=<rot 0 -12.2949 0>}]
 dyn_hood_02_c [{shapeType=Capsule, length=8, radius=4, shapeName=dyn_hood_02_c_capsule, center=<2.6883 -0.5859 0>, rotation=<rot 0 -12.2948 0>}]
 dyn_hood_02_l [{shapeType=Capsule, length=8, radius=4, shapeName=dyn_hood_02_l_capsule, center=<2.6883 -0.5859 0>, rotation=<rot 0 -0.6775 0>}]
 dyn_hood_02_r [{shapeType=Capsule, length=8, radius=4, shapeName=dyn_hood_02_r_capsule, center=<2.6883 -0.5859 0>, rotation=<rot 0 -0.6775 0>}]
 dyn_hood_03_c [{shapeType=Capsule, length=8, radius=4, shapeName=dyn_hood_03_c_capsule, center=<2.6883 -0.5859 0>, rotation=<rot 0 -12.2948 0>}]
 dyn_hood_03_l [{shapeType=Capsule, length=8, radius=4, shapeName=dyn_hood_03_l_capsule, center=<2.6883 -0.5859 0>, rotation=<rot 0 -0.6775 0>}]
 dyn_hood_03_r [{shapeType=Capsule, length=8, radius=4, shapeName=dyn_hood_03_r_capsule, center=<2.6883 -0.5859 0>, rotation=<rot 0 -0.6775 0>}]
 dyn_hood_mstr_c [{shapeType=Capsule, length=8, radius=4, shapeName=dyn_hood_mstr_c_capsule}]
 dyn_hood_mstr_l [{shapeType=Capsule, length=8, radius=4, shapeName=dyn_hood_mstr_l_capsule}]
 dyn_hood_mstr_r [{shapeType=Capsule, length=8, radius=4, shapeName=dyn_hood_mstr_r_capsule}]
 head [{shapeType=Capsule, length=5.4469, radius=6.223, shapeName=head_capsule, rotation=<rot -90 0 0>}]
 spine_05 [{shapeType=Capsule, length=9.1557, radius=11.6539, shapeName=spine_05_capsule}]
 upperarm_l [{shapeType=Capsule, length=12.3271, radius=5.8431, shapeName=upperarm_l_capsule, center=<6 0 0>, rotation=<rot -90 0 0>}]
 upperarm_r [{shapeType=Capsule, length=12.3271, radius=5.8431, shapeName=upperarm_r_capsule, center=<-6 0 0>, rotation=<rot -90 0 0>}]
}
constraints [
 {bone1Name=dyn_hair_front_01_l, bone2Name=dyn_hair_front_mstr_l, swing1LimitDegrees=12, swing2LimitDegrees=12, twistLimitDegrees=2}
 {bone1Name=dyn_hair_front_02_l, bone2Name=dyn_hair_front_01_l, swing1LimitDegrees=12, swing2LimitDegrees=12, twistLimitDegrees=2}
 {bone1Name=dyn_hair_front_03_l, bone2Name=dyn_hair_front_02_l, swing1LimitDegrees=12, swing2LimitDegrees=12, twistLimitDegrees=2}
 {bone1Name=dyn_hair_front_01_r, bone2Name=dyn_hair_front_mstr_r, swing1LimitDegrees=12, swing2LimitDegrees=12, twistLimitDegrees=2}
 {bone1Name=dyn_hair_front_02_r, bone2Name=dyn_hair_front_01_r, swing1LimitDegrees=12, swing2LimitDegrees=12, twistLimitDegrees=2}
 {bone1Name=dyn_hair_front_03_r, bone2Name=dyn_hair_front_02_r, swing1LimitDegrees=12, swing2LimitDegrees=12, twistLimitDegrees=2}
 {bone1Name=dyn_hood_01_c, bone2Name=dyn_hood_mstr_c, swing1LimitDegrees=4, swing2LimitDegrees=2, twistLimitDegrees=2}
 {bone1Name=dyn_hood_02_c, bone2Name=dyn_hood_01_c, swing1LimitDegrees=4, swing2LimitDegrees=2, twistLimitDegrees=2}
 {bone1Name=dyn_hood_03_c, bone2Name=dyn_hood_02_c, swing1LimitDegrees=4, swing2LimitDegrees=2, twistLimitDegrees=2}
 {bone1Name=dyn_hood_01_l, bone2Name=dyn_hood_mstr_l, swing1LimitDegrees=8, swing2LimitDegrees=3, twistLimitDegrees=2}
 {bone1Name=dyn_hood_02_l, bone2Name=dyn_hood_01_l, swing1LimitDegrees=10, swing2LimitDegrees=3, twistLimitDegrees=2}
 {bone1Name=dyn_hood_03_l, bone2Name=dyn_hood_02_l, swing1LimitDegrees=10, swing2LimitDegrees=3, twistLimitDegrees=2}
 {bone1Name=dyn_hood_01_r, bone2Name=dyn_hood_mstr_r, swing1LimitDegrees=8, swing2LimitDegrees=3, twistLimitDegrees=2}
 {bone1Name=dyn_hood_02_r, bone2Name=dyn_hood_01_r, swing1LimitDegrees=10, swing2LimitDegrees=3, twistLimitDegrees=2}
 {bone1Name=dyn_hood_03_r, bone2Name=dyn_hood_02_r, swing1LimitDegrees=10, swing2LimitDegrees=3, twistLimitDegrees=2}
]
