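# Animation Blueprint for the CatSpace Skier face accessory cosmetic, following the parent mesh with added secondary motion.
# The copied parent pose runs through a Control Rig, head-space RigidBody physics, then spring-driven AnimDynamics on both ears.
# Each update refreshes gravity override and fades RigidBody alpha inversely to the SPR_Hood_R_Adjust curve.
# With a valid pawn, it also feeds material WPO deformation from limb transforms and enables cloth wind during skydive floating.
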AnimBlueprint /BRCosmetics/Characters/Player/Female/Medium/Bodies/F_MED_CatSpace_Skier/Meshes/Parts/F_MED_CatSpace_Skier_FaceAcc_AnimBP
target_skeleton=&/F_MED_CatSpace_Skier_FaceAcc_Skeleton
graphs [
 AnimGraph {nodes [
   Layered blend per bone(BlendPoses_0=Copy Pose From Mesh(SourceMeshComponent=Get Output() -> (Output:Skeletal Mesh Component Budgeted Object Reference)) {bUseAttachedParent=true}) {layerSetup [{branchFilters [
       {boneName=deform_pec_aim_base_l}
       {boneName=deform_pec_aim_l}
       {boneName=deform_pec_base_l}
       {boneName=deform_pec_l}
       {boneName=deform_pec_aim_base_r}
       {boneName=deform_pec_aim_r}
       {boneName=deform_pec_base_r}
       {boneName=deform_pec_r}
       {boneName=deform_upperarm_bk_base_l}
       {boneName=deform_upperarm_bk_l}
       {boneName=deform_upperarm_fr_base_l}
       {boneName=deform_upperarm_fr_l}
       {boneName=deform_upperarm_up_base_l}
       {boneName=deform_upperarm_up_l}
       {boneName=deform_upperarm_bk_base_r}
       {boneName=deform_upperarm_bk_r}
       {boneName=deform_upperarm_fr_base_r}
       {boneName=deform_upperarm_fr_r}
       {boneName=deform_upperarm_up_base_r}
       {boneName=deform_upperarm_up_r}
       {boneName=upperarm_twist_01_l}
       {boneName=upperarm_twist_01_r}
      ]}]}
   n1 = Control Rig(Source=Copy Pose From Mesh() {bUseAttachedParent=true}) {
    alphaBoolBlend {blendOption=Linear}
    controlRigAssetReference {blueprintRigClass=&/F_MED_CatSpace_Skier_FaceAcc_CtrlRig1.F_MED_CatSpace_Skier_FaceAcc_CtrlRig1_C}
   }
   n2 = Get Gravity Override()
   n3 = RigidBody!(ComponentPose=Local To Component(LocalPose=@n1), Alpha=Get ADS_Alpha() -> (ADS_Alpha:double), OverrideWorldGravity:Vector=@n2, ExternalForce) -> (Pose) {
    bOverrideWorldGravity=true
    overridePhysicsAsset=&/F_MED_CatSpace_Skier_FaceAcc_Physics
    simulationSpace=BaseBoneSpace
    alphaBoolBlend {blendOption=Linear}
    baseBoneRef {boneName=head}
    componentAppliedLinearAccClamp=<2500 2500 2500>
    componentLinearAccScale=<0.4 0.4 0.4>
    componentLinearVelScale=<0.4 0.4 0.4>
    simSpaceSettings {maxAngularAcceleration=100, maxAngularVelocity=5, maxLinearAcceleration=100, maxLinearVelocity=300, worldAlpha=0.4}
   }
   n4 = Anim Dynamics - Bone: dyn_ear_l!(ComponentPose=@n3, Alpha=1, GravityOverride:Vector=@n2, ExternalForce="0,0,0") -> (Pose) {
    angularDampingOverride=0.85
    angularSpringConstant=3500
    bAngularSpring=true
    bOverrideAngularDamping=true
    bUseGravityOverride=true
    simulationSpace=BoneRelative
    alphaBoolBlend {blendOption=Linear}
    boundBone {boneName=dyn_ear_l}
    componentAppliedLinearAccClamp=<2000 2000 2000>
    componentLinearAccScale=<0.4 0.4 0.4>
    componentLinearVelScale=<0.4 0.4 0.4>
    physicsBodyDefinitions [{
      boundBone {boneName=dyn_ear_l}
      boxExtents=<30 30 30>
      constraintSetup {angularLimitsMax=<6 6 12>, angularLimitsMin=<-6 -6 -20>, angularTarget=<1 0 0>}
      localJointOffset=<1.7 0 0>
     }]
    relativeSpaceBone {boneName=root}
   }
   n5 = Anim Dynamics - Bone: dyn_ear_r!(ComponentPose=@n4, Alpha=1, GravityOverride:Vector=@n2, ExternalForce="0,0,0") -> (Pose) {
    angularDampingOverride=0.85
    angularSpringConstant=3500
    bAngularSpring=true
    bOverrideAngularDamping=true
    bUseGravityOverride=true
    simulationSpace=BoneRelative
    alphaBoolBlend {blendOption=Linear}
    boundBone {boneName=dyn_ear_r}
    componentAppliedLinearAccClamp=<2000 2000 2000>
    componentLinearAccScale=<0.4 0.4 0.4>
    componentLinearVelScale=<0.4 0.4 0.4>
    physicsBodyDefinitions [{
      boundBone {boneName=dyn_ear_r}
      boxExtents=<35 35 35>
      constraintSetup {angularLimitsMax=<6 6 12>, angularLimitsMin=<-6 -6 -20>, angularTarget=<1 0 0>}
      localJointOffset=<-1 0 0>
     }]
    relativeSpaceBone {boneName=root}
   }
   Output Pose(Result=Component To Local(ComponentPose=@n5)) {layerGroup=None}
  ]}
 EventGraph {nodes [
   n1 = Event BlueprintUpdateAnimation()
   > n2 = Sequence!(execute) -> (then_0, then_1, then_2:Exec)
   > n3 = Gravity Override State Function DEPRECATED(CharacterPartAnimInstance=Self-Reference(), GameTimeInSeconds=@n1.DeltaTimeX, GravityOverrideParams=Get Gravity Override Params())
   > Set Gravity Override(Gravity Override=@n3)
   Set ADS_Alpha(
    execute=@n2.then_1
    ADS_Alpha:double=Remap Float(Input=GetCurveValue(CurveName=SPR_Hood_R_Adjust), OldRange=(X=0,Y=1), NewRange=(X=1,Y=0)) -> (Output=~)
   ) -> (Output_Get:double)
   Is Valid(exec=@n2.then_2, InputObject=Get playerPawn() -> (playerPawn:Fort Player Pawn Athena Object Reference))
   > Set WPO Globals Function(
    Material=Get matInstanceDynamic() -> (matInstanceDynamic:Material Instance Dynamic Object Reference)
    Overlay Material=Get overlayMID() -> (overlayMID:Material Instance Dynamic Object Reference)
    Skeletal Mesh=Get skelMesh() -> (skelMesh:Skeletal Mesh Component Object Reference)
    "Local Noise Transform  Location"="Get Local Noise Transform  Location"() -> ("Local Noise Transform  Location":Vector)
    Local Noise Offset=Get Local Noise Offset()
   )
   > Set WPO Regional Function!(
    execute
    Material:Material Instance Dynamic Object Reference=Get matInstanceDynamic()
    Overlay Material:Material Instance Dynamic Object Reference=Get overlayMID()
    Skeletal Mesh:Skeletal Mesh Component Object Reference=Get skelMesh()
    Velocity Transform:Name=thigh_l
    Velocity Interp Speed:double=10
    Collision Transform:Name=upperarm_l
    Collision Transform Axis:EAxis Enum
    Invert Collision Transform Axis:bool=false
    Delta Time X:double=@n1.DeltaTimeX
    Previous Velocity Transform Location:"Vector (by ref)"=Get R1 Velocity Transform Location()
    Previous Velocity Transform Velocity:"Vector (by ref)"=Get R1 Previous Velocity Transform Velocity() -> (R1 Previous Velocity Transform Velocity:Vector)
    Region:int=1
   ) -> (then)
   > Set WPO Regional Function(
    Material=Get matInstanceDynamic()
    Overlay Material=Get overlayMID()
    Skeletal Mesh=Get skelMesh()
    Velocity Transform=thigh_r
    Velocity Interp Speed=10
    Collision Transform=upperarm_r
    Invert Collision Transform Axis=false
    Delta Time X=@n1.DeltaTimeX
    Previous Velocity Transform Location=Get R2 Velocity Transform Location()
    Previous Velocity Transform Velocity=Get R2 Velocity Transform Velocity()
    Region=2
   )
   > n4 = Branch(Condition=Get bIsSkydiveFloating())
   > n5 = Set WPO Pho Wind Intensity(WPO Pho Wind Intensity=1)
   SetScalarParameterValue(execute=[@n6.then, @n5.then], self=Get matInstanceDynamic(), ParameterName=ClothSim_WindStrength, Value=Get WPO Pho Wind Intensity())
   n6 = Set WPO Pho Wind Intensity(execute=@n4.else, WPO Pho Wind Intensity=0)
   Event BlueprintInitializeAnimation()
   > Set SkelMesh(SkelMesh=GetOwnerSkeletalMeshComponent())
   n7 = TryGetPawnOwner()
   > Is Valid(InputObject=@n7)
   > n8 = Cast To FortPlayerPawnAthena(Object=@n7)
   > Set playerPawn(playerPawn:Fort Player Pawn Athena Object Reference=@n8) -> (Output_Get:Fort Player Pawn Athena Object Reference)
   Event OnCharacterCustomizationLoaded()
   > n9 = Set skelMesh(skelMesh:Skeletal Mesh Component Object Reference=GetPartSkeletalMeshComponent()) -> (Output_Get:Skeletal Mesh Component Object Reference)
   > n10 = Get Dynamic Material Instance(SkeletalMeshComponent=@n9)
   > Set matInstanceDynamic(matInstanceDynamic:Material Instance Dynamic Object Reference=@n10) -> (Output_Get:Material Instance Dynamic Object Reference)
   > n11 = GetOrCreateDynamicOverlayMaterial(MeshComponent=@n9, MaterialName=None)
   > Set overlayMID(overlayMID:Material Instance Dynamic Object Reference=@n11) -> (Output_Get:Material Instance Dynamic Object Reference)
   TryGetPawnOwner()
  ]}
]
variables [
 ADS_Alpha {type=number}
 Gravity Override {type=Vector}
 Gravity Override Params {type=GravityOverrideParamsStruct}
 Local Noise Offset {type=Vector}
 "Local Noise Transform  Location" {type=Vector}
 matInstanceDynamic {type=MaterialInstanceDynamic}
 Output {type=SkeletalMeshComponentBudgeted}
 overlayMID {type=MaterialInstanceDynamic}
 playerPawn {type=FortPlayerPawnAthena}
 R1 Previous Velocity Transform Velocity {type=Vector}
 R1 Velocity Transform Location {type=Vector}
 R2 Velocity Transform Location {type=Vector}
 R2 Velocity Transform Velocity {type=Vector}
 SkelMesh {type=SkeletalMeshComponent}
 WPO Pho Wind Intensity {type=number}
]
